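# Animation Blueprint for the NebulaBurden male medium head cosmetic, targeting the shared Fortnite base head skeleton.
# Its update event only forwards to the parent implementation, and the file shows no AnimGraph logic of its own.
# Loose nodes fetch the pawn owner and owning mesh, and assign a custom part mesh from the customization-loaded event to Output, with no execution link shown.
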AnimBlueprint /BRCosmetics/Characters/Player/Male/Medium/Heads/M_MED_NebulaBurden_Head/Meshes/M_MED_NebulaBurden_Head_AnimBP
target_skeleton=/Game/Characters/Player/Common/Fortnite_Base_Head/Fortnite_Base_Head_Export_Skeleton
graphs [EventGraph {nodes [
   n1 = Event BlueprintUpdateAnimation()
   > Parent: BlueprintUpdateAnimation!(execute, DeltaTimeX:float=@n1.DeltaTimeX) -> (then) {is=title K2Node_CallParentFunction}
   TryGetPawnOwner()
   GetOwnerSkeletalMeshComponent()
   Set Output(Output:Skeletal Mesh Component Budgeted Object Reference=Get Array Item with EFort Custom Part Type(Array=Event OnCharacterCustomizationLoaded().CustomMeshParts)) -> (Output_Get:Skeletal Mesh Component Budgeted Object Reference)
  ]}]
variables [Output {type=SkeletalMeshComponentBudgeted}]
