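# Post-process head AnimBlueprint for the NebulaBurden cosmetic, with procedural idle motion for four "Stone" offsets.
# Each update, game time drives sine and cosine waves at differing rates into per-stone translation and rotation variables; the Stone_04_Rotation setter is unconnected to execution.
# The AnimGraph chains additive bone-space Modify Bone nodes on dyn_hair_01–04 over the input pose, showing zero offsets and no visible link to those variables.
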AnimBlueprint /BRCosmetics/Characters/Player/Male/Medium/Heads/M_MED_NebulaBurden_Head/Meshes/M_MED_NebulaBurden_PostProcess_Head_AnimBP
target_skeleton=/Game/Characters/Player/Common/Fortnite_Base_Head/Fortnite_Base_Head_Export_Skeleton
graphs [
 AnimGraph {nodes [
   n1 = "Transform (Modify) Bone - Bone: dyn_hair_01"!(ComponentPose=Local To Component(LocalPose=Use cached pose 'IncomingPose'()), Alpha=1, Translation="0,0,0", Rotation="0,0,0") -> (Pose) {
    rotationMode=BMM_Additive
    rotationSpace=BCS_BoneSpace
    translationMode=BMM_Additive
    translationSpace=BCS_BoneSpace
    alphaBoolBlend {blendOption=Linear}
    boneToModify {boneName=dyn_hair_01}
   }
   n2 = "Transform (Modify) Bone - Bone: dyn_hair_02"!(ComponentPose=@n1, Alpha=1, Translation="0,0,0", Rotation="0,0,0") -> (Pose) {
    rotationMode=BMM_Additive
    rotationSpace=BCS_BoneSpace
    translationMode=BMM_Additive
    translationSpace=BCS_BoneSpace
    alphaBoolBlend {blendOption=Linear}
    boneToModify {boneName=dyn_hair_02}
   }
   n3 = "Transform (Modify) Bone - Bone: dyn_hair_03"!(ComponentPose=@n2, Alpha=1, Translation="0,0,0", Rotation="0,0,0") -> (Pose) {
    rotationMode=BMM_Additive
    rotationSpace=BCS_BoneSpace
    translationMode=BMM_Additive
    translationSpace=BCS_BoneSpace
    alphaBoolBlend {blendOption=Linear}
    boneToModify {boneName=dyn_hair_03}
   }
   n4 = "Transform (Modify) Bone - Bone: dyn_hair_04"!(ComponentPose=@n3, Alpha=1, Translation="0,0,0", Rotation="0,0,0") -> (Pose) {
    rotationMode=BMM_Additive
    rotationSpace=BCS_BoneSpace
    translationMode=BMM_Additive
    translationSpace=BCS_BoneSpace
    alphaBoolBlend {blendOption=Linear}
    boneToModify {boneName=dyn_hair_04}
   }
   Output Pose(Result=Component To Local(ComponentPose=@n4))
   Save cached pose 'IncomingPose'(Pose=Input Pose()) {cacheName=IncomingPose}
  ]}
 EventGraph {nodes [
   Event BlueprintUpdateAnimation()
   > Set TimeInSeconds(TimeInSeconds:double=GetGameTimeInSeconds()) -> (Output_Get:double)
   > n1 = Sequence!(execute) -> (then_0, then_1, then_2:Exec, then_3:Exec)
   n2 = float * float(A=Get TimeInSeconds() -> (TimeInSeconds:double), B=1) {is=title K2Node_PromotableOperator}
   n3 = float * float(A="Sin (Radians)"(A=float * float(A=1.21, B=@n2)), B=0.1) {is=title K2Node_CommutativeAssociativeBinaryOperator}
   n4 = float * float(A="Sin (Radians)"(A=float * float(A=0.56, B=@n2)), B=0.8)
   n5 = float * float(A="Sin (Radians)"(A=float * float(A=0.8, B=@n2)), B=0.6)
   Set Stone_04_Translation(execute=@n1.then_3, Stone_04_Translation:Vector=vector * vector(A=MakeVector(X=@n4, Y=@n3, Z=@n5), B:float=0.5) -> (ReturnValue=~)) -> (Output_Get:Vector)
   n6 = float * float(A=Get TimeInSeconds(), B=1.4) {is=title K2Node_PromotableOperator}
   n7 = float * float(A="Cos (Radians)"!(A:double=float * float(A=0.78, B=@n6)) -> (ReturnValue:double), B=1.67) {is=title K2Node_CommutativeAssociativeBinaryOperator}
   n8 = float * float(A="Sin (Radians)"(A=float * float(A=0.76, B=@n6)), B=0.3)
   n9 = float * float(A="Cos (Radians)"(A=float * float(A=0.76, B=@n6)), B=0.68)
   Set Stone_03_Translation(execute=@n1.then_2, Stone_03_Translation:Vector=vector * vector(A=MakeVector(X=@n9, Y=@n8, Z=@n7), B=0.3) -> (ReturnValue=~)) -> (Output_Get:Vector)
   n10 = float * float(A=Get TimeInSeconds(), B=3) {is=title K2Node_PromotableOperator}
   n11 = "Sin (Radians)"(A=float * float(A=0.16, B=@n10) {is=title K2Node_CommutativeAssociativeBinaryOperator})
   n12 = "Sin (Radians)"(A=float * float(A=0.46, B=@n10))
   n13 = "Sin (Radians)"(A=float * float(A=0.24, B=@n10))
   > Set Stone_03_Rotation(Stone_03_Rotation:Rotator=MakeRotator(Roll=float * float(A=@n13, B=12), Pitch=float * float(A=@n12, B=11), Yaw=float * float(A=@n11, B=14))) -> (Output_Get:Rotator)
   n14 = float * float(A=Get TimeInSeconds(), B=2.1) {is=title K2Node_PromotableOperator}
   n15 = float * float(A="Sin (Radians)"(A=float * float(A=0.67, B=@n14)), B=1.5) {is=title K2Node_CommutativeAssociativeBinaryOperator}
   n16 = float * float(A="Sin (Radians)"(A=float * float(A=1.06, B=@n14)), B=0.12)
   n17 = float * float(A="Sin (Radians)"(A=float * float(A=0.45, B=@n14)), B=0.4)
   Set Stone_02_Translation(execute=@n1.then_1, Stone_02_Translation:Vector=vector * vector(A=MakeVector(X=@n17, Y=@n16, Z=@n15), B=0.1) -> (ReturnValue=~)) -> (Output_Get:Vector)
   n18 = float * float(A=Get TimeInSeconds(), B=3) {is=title K2Node_PromotableOperator}
   n19 = "Sin (Radians)"(A=float * float(A=0.3, B=@n18) {is=title K2Node_CommutativeAssociativeBinaryOperator})
   n20 = "Sin (Radians)"(A=float * float(A=0.1, B=@n18))
   n21 = "Sin (Radians)"(A=float * float(A=0.3, B=@n18))
   > Set Stone_02_Rotation(Stone_02_Rotation:Rotator=MakeRotator(Roll=float * float(A=@n21, B=13), Pitch=float * float(A=@n20, B=17), Yaw=float * float(A=@n19, B=2))) -> (Output_Get:Rotator)
   n22 = float * float(A=Get TimeInSeconds(), B=3) {is=title K2Node_PromotableOperator}
   n23 = float * float(A="Cos (Radians)"(A=float * float(A=0.7, B=@n22)), B=1.1) {is=title K2Node_CommutativeAssociativeBinaryOperator}
   n24 = float * float(A="Sin (Radians)"(A=float * float(A=1.1, B=@n22)), B=0.2)
   n25 = float * float(A="Cos (Radians)"(A=float * float(A=0.61, B=@n22)), B=0.9)
   Set Stone_01_Translation(execute=@n1.then_0, Stone_01_Translation:Vector=vector * vector(A=MakeVector(X=@n25, Y=@n24, Z=@n23), B=0.25) -> (ReturnValue=~)) -> (Output_Get:Vector)
   n26 = float * float(A=Get TimeInSeconds(), B=3) {is=title K2Node_PromotableOperator}
   n27 = "Sin (Radians)"(A=float * float(A=0.5, B=@n26) {is=title K2Node_CommutativeAssociativeBinaryOperator})
   n28 = "Sin (Radians)"(A=float * float(A=0.2, B=@n26))
   n29 = "Sin (Radians)"(A=float * float(A=0.6, B=@n26))
   > Set Stone_01_Rotation(Stone_01_Rotation:Rotator=MakeRotator(Roll=float * float(A=@n29, B=8), Pitch=float * float(A=@n28, B=11), Yaw=float * float(A=@n27, B=1.1))) -> (Output_Get:Rotator)
   n30 = float * float(A=Get TimeInSeconds(), B=3) {is=title K2Node_PromotableOperator}
   n31 = "Sin (Radians)"(A=float * float(A=0.32, B=@n30) {is=title K2Node_CommutativeAssociativeBinaryOperator})
   n32 = "Sin (Radians)"(A=float * float(A=0.7, B=@n30))
   n33 = "Sin (Radians)"(A=float * float(A=0.35, B=@n30))
   Set Stone_04_Rotation(Stone_04_Rotation:Rotator=MakeRotator(Roll=float * float(A=@n33, B=9), Pitch=float * float(A=@n32, B=6), Yaw=float * float(A=@n31, B=16))) -> (Output_Get:Rotator)
  ]}
]
variables [
 Stone_01_Rotation {type=Rotator}
 Stone_01_Translation {type=Vector}
 Stone_02_Rotation {type=Rotator}
 Stone_02_Translation {type=Vector}
 Stone_03_Rotation {type=Rotator}
 Stone_03_Translation {type=Vector}
 Stone_04_Rotation {type=Rotator}
 Stone_04_Translation {type=Vector}
 TimeInSeconds {type=number}
]
